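# Translucent unlit particle material for the dance grenade's bits, usable on Niagara and Cascade sprites.
# Particle color feeds an emissive adjustment function, while alpha from an 8x8 husk-bits SubUV flipbook, advanced by particle relative time, forms a mask.
# A saturated, offset sine of particle lifetime pulses that mask.
# Shading path switches give mobile a fixed mid-sequence frame and no pulse.
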Material /Figment_Athena_DanceGrenade/Gameplay/M_DanceGrenade_Bits
graph {
 properties {
  bEnableResponsiveAA=true
  bUsedWithNiagaraSprites=true
  bUsedWithParticleSprites=true
  blendMode=BLEND_Translucent
  pixelDepthOffsetMode=PDOM_Legacy
  shadingModel=MSM_Unlit
  usedShadingModels=Unlit
 }
 nodes [
  MaterialFunctionCall(Emissive=ParticleColor().RGB) {materialFunction=&MF_FX_EmissiveAdjustments}
  n1 = Multiply(A=ParticleRelativeTime(), B=Constant() {r=64})
  n2 = Constant() {r=32}
  n3 = ParticleRelativeTime()
  n4 = Add(A=Sine(None=Multiply(A=@n3) {constB=3})) {constB=0.8}
  Multiply(
   A=MaterialFunctionCall!(
    Texture=TextureObject() {texture=/Game/Effects/Fort_Effects/Textures/Husk/T_Husk_Bits_SubUV}
    UVs
    SubImages=Constant2Vector() {g=8, r=8}
    Frame=ShadingPathSwitch(Default=@n1, Mobile=@n2) {desc=Migrated from FeatureLevelSwitch}
   ) {materialFunction=/Engine/Functions/Engine_MaterialFunctions03/Texturing/SubUV_Function}.Alpha
   B=ShadingPathSwitch(Default=Saturate(None=@n4), Mobile=Constant() {r=1}) {desc=Migrated from FeatureLevelSwitch}
  )
 ]
}
